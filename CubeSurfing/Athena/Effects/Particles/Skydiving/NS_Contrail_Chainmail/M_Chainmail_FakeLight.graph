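# Additive unlit Niagara sprite material that fakes a light on scene surfaces for the Chainmail skydiving contrail.
# With depth test off, it measures distance from the world position behind translucency to the particle center, scaled by radius and stretch, for a steep fourth-power falloff.
# That distance also hue-shifts the color, unless particle color is switched on.
# A time-of-day function scales emissive, and a stretch parameter enlarges the sprite around its center.
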Material /CubeSurfing/Athena/Effects/Particles/Skydiving/NS_Contrail_Chainmail/M_Chainmail_FakeLight
graph {
 properties {
  bDisableDepthTest=true
  bUsedWithNiagaraSprites=true
  blendMode=BLEND_Additive
  pixelDepthOffsetMode=PDOM_Legacy
  shadingModel=MSM_Unlit
  translucencyPass=MTP_BeforeDOF
  usedShadingModels=Unlit
 }
 nodes [
  n1 = ParticleColor()
  n2 = VectorParameter() {parameterName="Color && Hue Shift", defaultValue=<rgba 0 2.6667 4 0.11>}
  n3 = MaterialFunctionCall() {materialFunction=&WorldPositionBehindTranslucency}
  n4 = ParticlePositionWS()
  n5 = ParticleRadius()
  n6 = ScalarParameter() {defaultValue=1, parameterName=Particle Stretch}
  n7 = Add(A=@n6)
  n8 = Saturate(None=Divide(A=Distance(A=@n3.Result, B=@n4), B=Multiply(A=@n5, B=@n7) {constB=2}))
  n9 = MaterialFunctionCall(Hue Shift Percentage=Multiply(A=@n2.A, B=@n8), Texture=@n2.RGB) {materialFunction=&HueShift}
  n10 = VectorParameter() {parameterName=TOD Emissive, defaultValue=<rgba 1 1 1 1>}
  MaterialFunctionCall(Emissive=Multiply(
    A=StaticSwitchParameter(True=@n1.RGB, False=@n9.Result) {parameterName=Use Particle Color}
    B=MaterialFunctionCall(Value_Morning=@n10.R, Value_Day=@n10.R, Value_Evening=@n10.R, Value_Night=@n10.R) {materialFunction=&Scalar_Time_of_Day}.Result
   )) {materialFunction=&MF_FX_EmissiveAdjustments}
  n11 = WorldPosition()
  Reroute(None=Multiply(A=Subtract(A=@n11.XYZ, B=@n4), B=@n6))
  n12 = OneMinus(None=@n8)
  n13 = Multiply(A=@n12, B=@n12)
  n14 = Multiply(A=@n13, B=@n13)
  StaticSwitchParameter(True=Multiply(A=@n1.A, B=@n14), False=@n14) {parameterName=Use Particle Color}
 ]
}
